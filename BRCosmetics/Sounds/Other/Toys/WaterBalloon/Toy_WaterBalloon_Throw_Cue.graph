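# Sound cue for throwing the water balloon toy, playing one of five throw variations per trigger.
# A Random node with equal weights picks among five wave players, each holding its own throw sample, and feeds the output directly.
# It uses emote attenuation and the emote music sound class, and overrides concurrency with a rule that blocks new instances once the limit is reached.
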SoundCue /BRCosmetics/Sounds/Other/Toys/WaterBalloon/Toy_WaterBalloon_Throw_Cue
root=Random_0
nodes [
 {id=Random_0, children [WavePlayer_0,WavePlayer_1,WavePlayer_2,WavePlayer_3,WavePlayer_4], properties {weights [1,1,1,1,1]}}
 {id=WavePlayer_0, properties {soundWaveAssetPtr=&/WaterBalloon_Throw_01}}
 {id=WavePlayer_1, properties {soundWaveAssetPtr=&/WaterBalloon_Throw_02}}
 {id=WavePlayer_2, properties {soundWaveAssetPtr=&/WaterBalloon_Throw_03}}
 {id=WavePlayer_3, properties {soundWaveAssetPtr=&/WaterBalloon_Throw_04}}
 {id=WavePlayer_4, properties {soundWaveAssetPtr=&/WaterBalloon_Throw_05}}
]
properties {
 attenuationSettings=&Fort_Emotes_Athena_Att
 bOverrideConcurrency=true
 maxDistance=1600
 soundClassObject=&Fort_Music_Emote
 volumeMultiplier=1
 concurrencyOverrides {resolutionRule=PreventNew}
}
